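# Animation Blueprint for the male medium Multibot player body, layering secondary motion onto the pose copied from its attached parent mesh.
# The pose runs through a Control Rig fed ads_alpha, RigidBody collar physics, and a weighted backpack constraint between two spine bones.
# While targeting, ads_alpha also drives interpolated additive scale adjustments on right shoulder bones.
# Weapon, customization and skydiving events toggle the "Hide Element 07" material parameter, checking for gauntlet pickaxes.
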AnimBlueprint /Game/Characters/Player/Male/Medium/Bodies/M_MED_Multibot/Meshes/M_MED_Multibot_AnimBP
target_skeleton=&/M_MED_Multibot_Skeleton
graphs [
 AnimGraph {nodes [
   n1 = Control Rig!(Source=Copy Pose From Mesh() {bUseAttachedParent=true}, Alpha, Ads_Float:double=Get ads_alpha() -> (ads_alpha:double)) -> (Pose) {
    alphaBoolBlend {blendOption=Linear}
    controlRigAssetReference {blueprintRigClass=&/M_MED_Multibot_CtrlRig.M_MED_Multibot_CtrlRig_C}
   }
   n2 = RigidBody(ComponentPose=Local To Component(LocalPose=@n1)) {
    bClampLinearTranslationLimitToRefPose=true
    bOverrideWorldGravity=true
    overridePhysicsAsset=&/M_MED_Multibot_Collar_Physics
    simulationSpace=BaseBoneSpace
    alphaBoolBlend {blendOption=Linear}
    baseBoneRef {boneName=spine_05}
    componentAppliedLinearAccClamp=<1000 1000 1100>
    componentLinearAccScale=<0.5 0.5 0.5>
    componentLinearVelScale=<0.5 0.5 0.5>
    overrideWorldGravity {z=-980}
   }
   n3 = "Constraint - dyn_backpack (PP)"!(ComponentPose=@n2, Alpha=1, ConstraintWeights_0=0.8734, ConstraintWeights_1:float=0.1266) -> (Pose) {
    alphaBoolBlend {blendOption=Linear}
    boneToModify {boneName=dyn_backpack}
    constraintSetup [{targetBone {boneName=spine_03}}, {targetBone {boneName=spine_05}}]
    constraintWeights [1,1]
   }
   n4 = Get ads_alpha()
   n5 = "Transform (Modify) Bone - Bone: dyn_upperarm_offset_r"(ComponentPose=@n3, Alpha=@n4, Translation="0,0,0", Rotation="0,0,0", Scale="1,0.8,0.95") {
    rotationMode=BMM_Additive
    rotationSpace=BCS_BoneSpace
    scaleMode=BMM_Additive
    scaleSpace=BCS_BoneSpace
    translationMode=BMM_Additive
    translationSpace=BCS_BoneSpace
    alphaBoolBlend {blendOption=Linear}
    alphaScaleBiasClamp {bInterpResult=true, interpSpeedDecreasing=50, interpSpeedIncreasing=50}
    boneToModify {boneName=dyn_upperarm_offset_r}
    scale {y=0.8, z=0.95}
   }
   n6 = "Transform (Modify) Bone - Bone: dyn_shldr_base_r"(ComponentPose=@n5, Alpha=@n4, Translation="0,0,0", Rotation="0,0,0", Scale="0.875,0.875,0.875") {
    rotationMode=BMM_Additive
    rotationSpace=BCS_BoneSpace
    scaleMode=BMM_Additive
    scaleSpace=BCS_BoneSpace
    translationMode=BMM_Additive
    translationSpace=BCS_BoneSpace
    alphaBoolBlend {blendOption=Linear}
    alphaScaleBiasClamp {bInterpResult=true, interpSpeedDecreasing=50, interpSpeedIncreasing=50}
    boneToModify {boneName=dyn_shldr_base_r}
    scale=<0.875 0.875 0.875>
   }
   Output Pose(Result=Component To Local(ComponentPose=@n6)) {layerGroup=None}
  ]}
 EventGraph {nodes [
   Event BlueprintUpdateAnimation()
   > n1 = Branch(Condition=Get bIsTargeting())
   > Set ads_alpha(ads_alpha:double=1) -> (Output_Get:double)
   Set ads_alpha(execute=@n1.else, ads_alpha=0)
   n2 = Event OnWeaponChanged()
   n3 = Get Dynamic Material Instance(execute=[@n2.then, @n11.then], SkeletalMeshComponent=GetPartSkeletalMeshComponent())
   > n4 = Sequence()
   n5 = PickaxeItemDefinitionHasAnyMetaTag(Tags=Make Array("[0]"=(TagName="Cosmetics.Pickaxe.Gauntlet")), bRequiresEquipped=true)
   > n6 = Branch(Condition=@n5)
   > n7 = Branch(Condition=Get bIsSkyDiving())
   SetScalarParameterValue(execute=[@n9.then, @n7.then, @n8.else, @n6.else], self=@n3, ParameterName=Hide Element 07, Value=0)
   SetScalarParameterValue(execute=[@n7.else, @n8.then], self=@n3, ParameterName=Hide Element 07, Value=1)
   n8 = Branch(execute=[@n10.then, @n4.then_1], Condition=@n5)
   n9 = Event OnBeginSkydiving()
   n10 = Event OnEndSkydiving()
   n11 = Event OnCharacterCustomizationLoaded()
  ]}
]
variables [ads_alpha {type=number}]
